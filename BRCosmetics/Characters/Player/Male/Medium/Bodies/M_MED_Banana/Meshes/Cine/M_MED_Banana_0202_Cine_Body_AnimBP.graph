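# Post-process-style Anim Blueprint for the cinematic Banana character body, layering procedural leg correctives and peel-flap motion onto the pose copied from the attached parent mesh.
# Thigh rotation relative to pelvis drives twist-bone shifts, cached before the flap stage.
# A right flap constraint feeds two RigidBody simulations, while the left flap constraint is unconnected.
# Targeting state drives interpolated alphas for the inner-flap simulation and a right-flap rotation override.
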AnimBlueprint /BRCosmetics/Characters/Player/Male/Medium/Bodies/M_MED_Banana/Meshes/Cine/M_MED_Banana_0202_Cine_Body_AnimBP
target_skeleton=&/M_MED_Banana_0202_Cine_Body_Skeleton
graphs [
 AnimGraph {nodes [
   "Constraint - dyn_flap_Base_l (PP)"!(ComponentPose, Alpha=1, ConstraintWeights_0=0.8, ConstraintWeights_1:float=0.2) -> (Pose) {
    alphaBoolBlend {blendOption=Linear}
    boneToModify {boneName=dyn_flap_Base_l}
    constraintSetup [{targetBone {boneName=spine_05}}, {targetBone {boneName=neck_01}}]
    constraintWeights [1,1]
   }
   n1 = "Constraint - dyn_flapBase_r (PP)"!(
    ComponentPose=Local To Component(LocalPose=Use cached pose 'PostLegShapes'())
    Alpha=1
    ConstraintWeights_0=0.8
    ConstraintWeights_1:float=0.2
   ) -> (Pose) {
    alphaBoolBlend {blendOption=Linear}
    boneToModify {boneName=dyn_flapBase_r}
    constraintSetup [{targetBone {boneName=spine_05}}, {targetBone {boneName=neck_01}}]
    constraintWeights [1,1]
   }
   n2 = RigidBody(ComponentPose=@n1) {
    bClampLinearTranslationLimitToRefPose=true
    bOverrideWorldGravity=true
    overridePhysicsAsset=/Game/Characters/Player/Male/Medium/Bodies/M_MED_Banana/Meshes/M_MED_Banana_Physics
    simulationSpace=BaseBoneSpace
    alphaBoolBlend {blendOption=Linear}
    baseBoneRef {boneName=root}
    componentAppliedLinearAccClamp=<800 800 800>
    componentLinearAccScale=<0.65 0.65 0.65>
    componentLinearVelScale=<0.3 0.3 0.3>
    overrideWorldGravity {z=-980}
   }
   n3 = RigidBody(ComponentPose=@n2, Alpha=Get adsRigidOff() -> (adsRigidOff:double)) {
    bClampLinearTranslationLimitToRefPose=true
    bOverrideWorldGravity=true
    overridePhysicsAsset=/Game/Characters/Player/Male/Medium/Bodies/M_MED_Banana/Meshes/M_MED_Banana_InnerFlap_Physics
    simulationSpace=BaseBoneSpace
    alphaBoolBlend {blendOption=Linear}
    alphaScaleBiasClamp {bInterpResult=true, interpSpeedDecreasing=2, interpSpeedIncreasing=2}
    baseBoneRef {boneName=root}
    componentAppliedLinearAccClamp=<800 800 800>
    componentLinearAccScale=<0.65 0.65 0.65>
    componentLinearVelScale=<0.3 0.3 0.3>
    overrideWorldGravity {z=-980}
   }
   n4 = "Transform (Modify) Bone - Bone: dyn_flapBase_r"(ComponentPose=@n3, Alpha=Get adsFlapOn() -> (adsFlapOn:double), Translation="0,0,0", Rotation="0,-20,0", Scale="1,1,1") {
    rotationMode=BMM_Replace
    rotationSpace=BCS_BoneSpace
    alphaBoolBlend {blendOption=Linear}
    alphaScaleBiasClamp {bInterpResult=true, interpSpeedDecreasing=3.5, interpSpeedIncreasing=8}
    boneToModify {boneName=dyn_flapBase_r}
    rotation {yaw=-20}
   }
   Output Pose(Result=Component To Local(ComponentPose=@n4)) {layerGroup=None}
   n5 = "Transform (Modify) Bone - Bone: thigh_twist_shift_l"(
    ComponentPose=Local To Component(LocalPose=Copy Pose From Mesh() {bCopyCurves=true, bUseAttachedParent=true})
    Alpha=Get LeftLegUP() -> (LeftLegUP:double)
    Translation="15,0,0"
    Rotation="0,0,0"
    Scale="1,1,1"
   ) {
    translationMode=BMM_Additive
    translationSpace=BCS_BoneSpace
    alphaBoolBlend {blendOption=Linear}
    boneToModify {boneName=thigh_twist_shift_l}
    translation {x=15}
   }
   n6 = "Transform (Modify) Bone - Bone: thigh_twist_shift_r"(ComponentPose=@n5, Alpha=Get RightLegUP() -> (RightLegUP:double), Translation="-15,0,0", Rotation="0,0,0", Scale="1,1,1") {
    translationMode=BMM_Additive
    translationSpace=BCS_BoneSpace
    alphaBoolBlend {blendOption=Linear}
    boneToModify {boneName=thigh_twist_shift_r}
    translation {x=-15}
   }
   Save cached pose 'PostLegShapes'(Pose=Component To Local(ComponentPose=@n6)) {cacheName=PostLegShapes}
  ]}
 EventGraph {nodes [
   Event BlueprintUpdateAnimation()
   n1 = GetOwnerSkeletalMeshComponent()
   n2 = BreakTransform(InTransform=GetDeltaTransformFromRefPose(self=@n1, BoneName=pelvis, BaseName=thigh_l))
   > Set LeftLegUP(LeftLegUP:double=MapRangeClamped(Value=BreakRotator(InRot=@n2.Rotation).Yaw, InRangeA=0, InRangeB=85)) -> (Output_Get:double)
   n3 = BreakTransform(InTransform=GetDeltaTransformFromRefPose(self=@n1, BoneName=pelvis, BaseName=thigh_r))
   > Set RightLegUP(RightLegUP:double=MapRangeClamped(Value=BreakRotator(InRot=@n3.Rotation).Yaw, InRangeA=0, InRangeB=85)) -> (Output_Get:double)
   n4 = "To Float (Boolean)"(InBool=Get bIsTargeting())
   > Set adsRigidOff(adsRigidOff:double=float - float(B=@n4) {is=title K2Node_CallFunction}) -> (Output_Get:double)
   > Set adsFlapOn(adsFlapOn:double=@n4) -> (Output_Get:double)
   TryGetPawnOwner()
  ]}
]
variables [
 adsFlapOn {type=number}
 adsRigidOff {type=number}
 LeftLegUP {type=number}
 RightLegUP {type=number}
]
